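# Translucent, two-sided unlit material, apparently an interaction highlight for the BasilOkra console in WildEstate.
# A custom primitive data parameter switches the tinted color between static dim, static bright and sine-pulsed bright, then divides it by eye adaptation to offset auto exposure.
# Opacity shapes a vertical gradient, erodes it with panning swirly noise above a minimum, and uses a UV mask to give the top its own opacity.
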Material /WildEstate_BasilOkra/Assets/Console/M_BasilOkra_ConsoleInteraction_P
graph {
 properties {bAutomaticallySetUsageInEditor=false, blendMode=BLEND_Translucent, shadingModel=MSM_Unlit, twoSided=true, usedShadingModels=Unlit}
 nodes [
  n1 = ScalarParameter() {bUseCustomPrimitiveData=true, desc="0 - Static Dim\r\n1 - Static Bright\r\n2 - Pulse Bright", parameterName=BP_CustomPrim_State, sliderMax=2}
  n2 = Saturate(None=VectorParameter() {parameterName=Color, defaultValue=<rgba 1 0.6724 0.0666 1>}.RGB)
  MaterialExpressionMultiply_8 = Multiply(A=@n2, B=ScalarParameter() {defaultValue=1.25, parameterName=StateA.Mult}) {constB=0.8}
  MaterialExpressionMultiply_9 = Multiply(A=@n2, B=ScalarParameter() {defaultValue=3, parameterName=StateB_C.Mult})
  n3 = Sine(None=Multiply(A=Time(), B=ScalarParameter() {defaultValue=1.5, parameterName=FlashSpeed}))
  MaterialExpressionMultiply_7 = Multiply(A=@MaterialExpressionMultiply_9, B=LinearInterpolate(Alpha=ConstantBiasScale(None=@n3)) {constA=0.5, constB=1.5})
  n4 = Switch!(
   SwitchValue=@n1
   Default
   StaticDIm=@MaterialExpressionMultiply_8
   StaticBright=@MaterialExpressionMultiply_9
   PulseBright=@MaterialExpressionMultiply_7
  ) {inputs [
    {
     inputName=StaticDIm
     input {
      expression=&self:MaterialExpressionMultiply_8
      inputName=None
      mask=0
      maskA=0
      maskB=0
      maskG=0
      maskR=0
      outputIndex=0
     }
    }
    {
     inputName=StaticBright
     input {
      expression=&self:MaterialExpressionMultiply_9
      inputName=None
      mask=0
      maskA=0
      maskB=0
      maskG=0
      maskR=0
      outputIndex=0
     }
    }
    {
     inputName=PulseBright
     input {
      expression=&self:MaterialExpressionMultiply_7
      inputName=None
      mask=0
      maskA=0
      maskB=0
      maskG=0
      maskR=0
      outputIndex=0
     }
    }
   ]}
  Divide(A=@n4, B=EyeAdaptation().EyeAdaptation)
  n5 = MaterialFunctionCall() {materialFunction=&LinearGradient}
  n6 = Multiply(A=Saturate(None=@n5.VGradient), B=OneMinus(None=Power(Base=@n5.VGradient) {constExponent=6}))
  n7 = Panner() {speedX=0.03, speedY=1}
  n8 = Constant2Vector() {g=0.2, r=0.3}
  n9 = Multiply(
   A=TextureSample(UVs=Multiply(A=@n7, B=@n8)) {texture=/Game/Weapons/WeaponSkins/Textures/T_Swirly_Noise_D1}.R
   B=ScalarParameter() {defaultValue=0.3, parameterName=NoiseSub}
  ) {constB=0.2}
  LinearInterpolate(
   A=ScalarParameter() {defaultValue=0.05, parameterName=TopOpacity}
   B=Max(A=Subtract(A=@n6, B=@n9), B=ScalarParameter() {defaultValue=0.02, parameterName=MinSideOpacity})
   Alpha=NamedRerouteUsage()
  )
  n10 = TextureCoordinate()
  NamedRerouteDeclaration(None=SmoothStep(Value=ComponentMask(None=@n10) {g=true}) {constMax=0.05, constMin=0.01}) {name=TopMask, nodeColor=<rgba 1 0 0.9844 1>}
 ]
}
